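# Physics asset for the dynamic wing pieces on the right shoe of the female medium Tiger Fashion character cosmetic.
# A small capsule body on the ball_r foot bone anchors two thin box bodies on the dyn_wings_r bones.
# Each wing is constrained directly to the foot rather than chained, with tight twist and first swing limits and a wider second swing for restrained flapping.
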PhysicsAsset /BRCosmetics/Characters/Player/Female/Medium/Bodies/F_MED_Tiger_Fashion/Meshes/F_MED_Tiger_Fashion_Dyn_rShoes
bodies {
 ball_r [{shapeType=Capsule, length=1, radius=1}]
 dyn_wings_r_01 [{shapeType=Box, extentX=4, extentY=6, extentZ=1, center=<1.5 0 0>}]
 dyn_wings_r_02 [{shapeType=Box, extentX=4, extentY=6, extentZ=1, center=<1.5 0 0>}]
}
constraints [
 {bone1Name=dyn_wings_r_01, bone2Name=ball_r, swing1LimitDegrees=5, swing2LimitDegrees=25, twistLimitDegrees=2}
 {bone1Name=dyn_wings_r_02, bone2Name=ball_r, swing1LimitDegrees=5, swing2LimitDegrees=25, twistLimitDegrees=2}
]
