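# Post-pose Anim Blueprint for the F_MED_Tiger_Fashion cosmetic body, adding secondary motion to the pose copied from its attached parent mesh.
# A Control Rig feeds a weighted spine constraint on the zipper master bone, then AnimDynamics swings the zipper.
# Per-foot RigidBody nodes simulate shoe dynamics in foot bone space.
# The event graph maps thigh yaw to glute offset alphas, but those Transform Bone nodes are not wired to the output.
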AnimBlueprint /BRCosmetics/Characters/Player/Female/Medium/Bodies/F_MED_Tiger_Fashion/Meshes/F_MED_Tiger_Fashion_AnimBP
target_skeleton=&/F_MED_Tiger_Fashion_Skeleton
graphs [
 AnimGraph {nodes [
   n1 = "Transform (Modify) Bone - Bone: dyn_glute_l"(Alpha=Get thigh_rot_up_l() -> (thigh_rot_up_l:double), Translation="-4,-6,0.000", Rotation="0,0,0", Scale="1,1,1") {
    translationMode=BMM_Additive
    translationSpace=BCS_BoneSpace
    alphaBoolBlend {blendOption=Linear}
    boneToModify {boneName=dyn_glute_l}
    translation=<-4 -6>
   }
   "Transform (Modify) Bone - Bone: dyn_glute_r"(ComponentPose=@n1, Alpha=Get thigh_rot_up_r() -> (thigh_rot_up_r:double), Translation="-4,-6,0", Rotation="0,0,0", Scale="1,1,1") {
    translationMode=BMM_Additive
    translationSpace=BCS_BoneSpace
    alphaBoolBlend {blendOption=Linear}
    boneToModify {boneName=dyn_glute_r}
    translation=<-4 -6>
   }
   n2 = Control Rig(Source=Copy Pose From Mesh() {bUseAttachedParent=true}) {
    alphaBoolBlend {blendOption=Linear}
    controlRigAssetReference {blueprintRigClass=&/F_MED_Tiger_Fashion_CtrlRig1.F_MED_Tiger_Fashion_CtrlRig1_C}
   }
   n3 = "Constraint - dyn_zipper_MSTR (PPPP)"!(
    ComponentPose=Local To Component(LocalPose=@n2)
    Alpha=1
    ConstraintWeights_0=0.066
    ConstraintWeights_1:float=0.077
    ConstraintWeights_2:float=0.743
    ConstraintWeights_3:float=0.114
   ) -> (Pose) {
    alphaBoolBlend {blendOption=Linear}
    boneToModify {boneName=dyn_zipper_MSTR}
    constraintSetup [{targetBone {boneName=spine_02}}, {targetBone {boneName=spine_03}}, {targetBone {boneName=spine_04}}, {targetBone {boneName=spine_05}}]
    constraintWeights [1,1,1,1]
   }
   n4 = Anim Dynamics - Bone: dyn_zipper(ComponentPose=@n3, Alpha=1) {
    angularDampingOverride=0.995
    bGravityOverrideInSimSpace=true
    bOverrideAngularDamping=true
    bUseGravityOverride=true
    alphaBoolBlend {blendOption=Linear}
    boundBone {boneName=dyn_zipper}
    componentAppliedLinearAccClamp=<3000 3000 3000>
    componentLinearAccScale=<0.3 0.3 0.3>
    componentLinearVelScale=<0.3 0.3 0.3>
    gravityOverride {z=-980}
    physicsBodyDefinitions [{
      sphereCollisionRadius=0
      boundBone {boneName=dyn_zipper}
      boxExtents=<5 3 3>
      constraintSetup {angularLimitsMax=<0 10 15>, angularLimitsMin=<0 -45 -15>, angularTarget=<0 0 0>}
      localJointOffset=<1 0 0>
     }]
   }
   n5 = RigidBody(ComponentPose=@n4) {
    bOverrideWorldGravity=true
    overridePhysicsAsset=&/F_MED_Tiger_Fashion_Dyn_lShoes
    simulationSpace=BaseBoneSpace
    alphaBoolBlend {blendOption=Linear}
    baseBoneRef {boneName=foot_l}
    componentAppliedLinearAccClamp=<3000 3000 3000>
    componentLinearAccScale {x=0.7, z=0.7}
    componentLinearVelScale {x=0.3, z=0.3}
    overrideWorldGravity {z=-980}
   }
   n6 = RigidBody(ComponentPose=@n5) {
    bOverrideWorldGravity=true
    overridePhysicsAsset=&/F_MED_Tiger_Fashion_Dyn_rShoes
    simulationSpace=BaseBoneSpace
    alphaBoolBlend {blendOption=Linear}
    baseBoneRef {boneName=foot_r}
    componentAppliedLinearAccClamp=<3000 3000 3000>
    componentLinearAccScale {x=0.8, z=0.8}
    componentLinearVelScale {x=0.3, z=0.3}
    overrideWorldGravity {z=-980}
   }
   Output Pose(Result=Component To Local(ComponentPose=@n6)) {layerGroup=None}
  ]}
 EventGraph {nodes [
   Event BlueprintUpdateAnimation()
   n1 = BreakTransform(InTransform=GetDeltaTransformFromRefPose(self=GetOwnerSkeletalMeshComponent(), BoneName=thigh_r, BaseName=pelvis))
   > Set thigh_rot_up_r(thigh_rot_up_r:double=MapRangeClamped(Value=BreakRotator(InRot=@n1.Rotation).Yaw, InRangeA=0, InRangeB=-100)) -> (Output_Get:double)
   n2 = BreakTransform(InTransform=GetDeltaTransformFromRefPose(self=GetOwnerSkeletalMeshComponent(), BoneName=thigh_l, BaseName=pelvis))
   > Set thigh_rot_up_l(thigh_rot_up_l:double=MapRangeClamped(Value=BreakRotator(InRot=@n2.Rotation).Yaw, InRangeA=-30, InRangeB=-100)) -> (Output_Get:double)
  ]}
]
variables [
 thigh_rot_up_l {type=number}
 thigh_rot_up_r {type=number}
]
