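# Physics asset for the Tiger Fashion female medium face accessory, defining secondary-motion bodies for temple braids, ears, hat back, temple fur and eye fur.
# Each braid is a seven-body chain from the head with limited swing and tight twist; ears form two-link chains and temple fur branches from a base.
# Capsules on head, neck, upper spine, clavicles and upper arms provide body volumes, linked by free-rotation constraints.
# Left and right braids differ slightly in upper shapes and limits.
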PhysicsAsset /BRCosmetics/Characters/Player/Female/Medium/Bodies/F_MED_Tiger_Fashion/Meshes/Parts/F_MED_Tiger_Fashion_FaceAcc_TempleBraid_Physics1
bodies {
 clavicle_l [{shapeType=Capsule, length=11.1237, radius=9, center=<4.1434 -0.4322 -0.1613>, rotation=<rot 0.4354 -94.9807 -85.0169>}]
 clavicle_r [{shapeType=Capsule, length=10.1237, radius=9, center=<-4.966 0.3984 0.9285>, rotation=<rot 0.7971 -99.1489 -85.0618>}]
 dyn_ear_l_01 [{shapeType=Box, extentX=5, extentY=6, extentZ=2, center=<2 0 0>}]
 dyn_ear_l_02 [{shapeType=Box, extentX=5, extentY=6, extentZ=2, center=<2 0 0>}]
 dyn_ear_r_01 [{shapeType=Box, extentX=5, extentY=6, extentZ=2, center=<2 0 0>}]
 dyn_ear_r_02 [{shapeType=Box, extentX=5, extentY=6, extentZ=2, center=<2 0 0>}]
 dyn_eyeFur_l [{shapeType=Capsule, length=1, radius=1, center=<0.5 0 0>, rotation=<rot 90 0 0>}]
 dyn_eyeFur_r [{shapeType=Capsule, length=1, radius=1, center=<0.5 0 0>, rotation=<rot 90 0 0>}]
 dyn_hat_back [{shapeType=Box, extentX=6, extentY=8, extentZ=2, center=<2.5 0 0>}]
 dyn_templeBraid_01_l [{shapeType=Box, extentX=6, extentY=4, extentZ=2, center=<3 0 0>}]
 dyn_templeBraid_01_r [{shapeType=Capsule, length=3.5, radius=1.3, shapeName=dyn_templeBraid_01_r_capsule, center=<2.7587 0 0>, rotation=<rot 90 0 0>}]
 dyn_templeBraid_02_l [{shapeType=Box, extentX=6, extentY=2, extentZ=2, center=<3 0 0>}]
 dyn_templeBraid_02_r [{shapeType=Capsule, length=4, radius=1, shapeName=dyn_templeBraid_02_r_capsule, center=<3.1977 0 0>, rotation=<rot 90 0 0>}]
 dyn_templeBraid_03_l [{shapeType=Capsule, length=6, radius=1, center=<3 0 0>, rotation=<rot 90 0 0>}]
 dyn_templeBraid_03_r [{shapeType=Capsule, length=6, radius=1, center=<3 0 0>, rotation=<rot 90 0 0>}]
 dyn_templeBraid_04_l [{shapeType=Capsule, length=6, radius=1, center=<3 0 0>, rotation=<rot 90 0 0>}]
 dyn_templeBraid_04_r [{shapeType=Capsule, length=6, radius=1, center=<3 0 0>, rotation=<rot 90 0 0>}]
 dyn_templeBraid_05_l [{shapeType=Capsule, length=3, radius=1.2, center=<1.5 0 0>, rotation=<rot 90 0 0>}]
 dyn_templeBraid_05_r [{shapeType=Capsule, length=3, radius=1.2, center=<1.5 0 0>, rotation=<rot 90 0 0>}]
 dyn_templeBraid_06_l [{shapeType=Capsule, length=3, radius=1.2, center=<1.5 0 0>, rotation=<rot 90 0 0>}]
 dyn_templeBraid_06_r [{shapeType=Capsule, length=3, radius=1.2, center=<1.5 0 0>, rotation=<rot 90 0 0>}]
 dyn_templeBraid_Base_l [{shapeType=Box, extentX=6, extentY=4, extentZ=2, center=<3 0 0>}]
 dyn_templeBraid_Base_r [{shapeType=Box, extentX=6, extentY=4, extentZ=2, center=<3 0 0>}]
 dyn_templeFur_base_l [{shapeType=Capsule, length=1, radius=2, center=<1.2 0 0>, rotation=<rot 90 0 0>}]
 dyn_templeFur_base_r [{shapeType=Capsule, length=1, radius=2, center=<1.2 0 0>, rotation=<rot 90 0 0>}]
 dyn_templeFur_l_01 [{shapeType=Capsule, length=1, radius=1, center=<0.5 0 0>, rotation=<rot 90 0 0>}]
 dyn_templeFur_l_02 [{shapeType=Capsule, length=1, radius=1, center=<0.5 0 0>, rotation=<rot 90 0 0>}]
 dyn_templeFur_l_03 [{shapeType=Capsule, length=1, radius=1, center=<0.5 0 0>, rotation=<rot 90 0 0>}]
 dyn_templeFur_r_01 [{shapeType=Capsule, length=1, radius=1, center=<0.5 0 0>, rotation=<rot 90 0 0>}]
 dyn_templeFur_r_02 [{shapeType=Capsule, length=1, radius=1, center=<0.5 0 0>, rotation=<rot 90 0 0>}]
 dyn_templeFur_r_03 [{shapeType=Capsule, length=1, radius=1, center=<0.5 0 0>, rotation=<rot 90 0 0>}]
 head [{shapeType=Capsule, length=6.5155, radius=5.9204, center=<1.5569 -0.3138 -0.0532>, rotation=<rot -0.3094 99.2805 -89.484>}]
 neck_01 [{shapeType=Capsule, length=0.01, radius=4.4254, rotation=<rot 0 -90 0>}]
 neck_02 [{shapeType=Capsule, length=0.01, radius=4.4254, rotation=<rot 0 -90 0>}]
 spine_04 [
  {shapeType=Capsule, length=6, radius=9.3, center=<1.2638 3.2098 0>}
  {shapeType=Capsule, length=8.7793, radius=4.0978, center=<-0.9955 10.3123 0.5058>, rotation=<rot -54.9996 0 2.4288>}
 ]
 spine_05 [
  {shapeType=Capsule, length=6.9062, radius=11.25, center=<-0.3385 -2.1504 0>}
  {shapeType=Capsule, length=15, radius=15, shapeName=spine_05_capsule, center=<9.003 -16.0263 0>}
 ]
 upperarm_twist_01_l [{shapeType=Capsule, length=21.3745, radius=5, center=<6 0 0>, rotation=<rot 90 0 0>}]
 upperarm_twist_01_r [{shapeType=Capsule, length=22.2034, radius=4, center=<-4.3118 1.0052 -0.1581>, rotation=<rot 89.5824 -19.2813 -22.3909>}]
}
constraints [
 {
  bone1Name=head
  bone2Name=neck_02
  swing1LimitDegrees=45
  swing1Motion=Free
  swing2LimitDegrees=45
  swing2Motion=Free
  twistLimitDegrees=45
  twistMotion=Free
 }
 {
  bone1Name=neck_02
  bone2Name=neck_01
  swing1LimitDegrees=45
  swing1Motion=Free
  swing2LimitDegrees=45
  swing2Motion=Free
  twistLimitDegrees=45
  twistMotion=Free
 }
 {bone1Name=dyn_templeBraid_01_l, bone2Name=dyn_templeBraid_Base_l, swing1LimitDegrees=25, swing2LimitDegrees=30, twistLimitDegrees=5}
 {bone1Name=dyn_templeBraid_02_l, bone2Name=dyn_templeBraid_01_l, swing1LimitDegrees=35, swing2LimitDegrees=45, twistLimitDegrees=5}
 {bone1Name=dyn_templeBraid_03_l, bone2Name=dyn_templeBraid_02_l, swing1LimitDegrees=35, swing2LimitDegrees=45, twistLimitDegrees=5}
 {bone1Name=dyn_templeBraid_04_l, bone2Name=dyn_templeBraid_03_l, swing1LimitDegrees=35, swing2LimitDegrees=45, twistLimitDegrees=5}
 {bone1Name=dyn_templeBraid_01_r, bone2Name=dyn_templeBraid_Base_r, swing1LimitDegrees=20, swing2LimitDegrees=30, twistLimitDegrees=5}
 {bone1Name=dyn_templeBraid_02_r, bone2Name=dyn_templeBraid_01_r, swing1LimitDegrees=30, swing2LimitDegrees=45, twistLimitDegrees=5}
 {bone1Name=dyn_templeBraid_03_r, bone2Name=dyn_templeBraid_02_r, swing1LimitDegrees=35, swing2LimitDegrees=45, twistLimitDegrees=5}
 {bone1Name=dyn_templeBraid_04_r, bone2Name=dyn_templeBraid_03_r, swing1LimitDegrees=35, swing2LimitDegrees=45, twistLimitDegrees=5}
 {
  bone1Name=neck_01
  bone2Name=spine_05
  swing1LimitDegrees=45
  swing1Motion=Free
  swing2LimitDegrees=45
  swing2Motion=Free
  twistLimitDegrees=45
  twistMotion=Free
 }
 {
  bone1Name=clavicle_l
  bone2Name=spine_05
  swing1LimitDegrees=45
  swing1Motion=Free
  swing2LimitDegrees=45
  swing2Motion=Free
  twistLimitDegrees=45
  twistMotion=Free
 }
 {
  bone1Name=clavicle_r
  bone2Name=spine_05
  swing1LimitDegrees=45
  swing1Motion=Free
  swing2LimitDegrees=45
  swing2Motion=Free
  twistLimitDegrees=45
  twistMotion=Free
 }
 {
  bone1Name=spine_05
  bone2Name=spine_04
  swing1LimitDegrees=45
  swing1Motion=Free
  swing2LimitDegrees=45
  swing2Motion=Free
  twistLimitDegrees=45
  twistMotion=Free
 }
 {bone1Name=dyn_templeBraid_Base_r, bone2Name=head, swing1LimitDegrees=20, swing2LimitDegrees=30, twistLimitDegrees=5}
 {bone1Name=dyn_templeBraid_Base_l, bone2Name=head, swing1LimitDegrees=25, swing2LimitDegrees=30, twistLimitDegrees=5}
 {bone1Name=dyn_templeBraid_05_l, bone2Name=dyn_templeBraid_04_l, swing1LimitDegrees=35, swing2LimitDegrees=45, twistLimitDegrees=5}
 {bone1Name=dyn_templeBraid_06_l, bone2Name=dyn_templeBraid_05_l, swing1LimitDegrees=35, swing2LimitDegrees=45, twistLimitDegrees=5}
 {bone1Name=dyn_templeBraid_05_r, bone2Name=dyn_templeBraid_04_r, swing1LimitDegrees=35, swing2LimitDegrees=45, twistLimitDegrees=5}
 {bone1Name=dyn_templeBraid_06_r, bone2Name=dyn_templeBraid_05_r, swing1LimitDegrees=35, swing2LimitDegrees=45, twistLimitDegrees=5}
 {bone1Name=dyn_hat_back, bone2Name=head, swing1LimitDegrees=20, swing2LimitDegrees=15, twistLimitDegrees=2}
 {bone1Name=dyn_ear_l_01, bone2Name=head, swing1LimitDegrees=10, swing2LimitDegrees=15, twistLimitDegrees=2}
 {bone1Name=dyn_ear_l_02, bone2Name=dyn_ear_l_01, swing1LimitDegrees=10, swing2LimitDegrees=15, twistLimitDegrees=2}
 {bone1Name=dyn_ear_r_01, bone2Name=head, swing1LimitDegrees=10, swing2LimitDegrees=15, twistLimitDegrees=2}
 {bone1Name=dyn_ear_r_02, bone2Name=dyn_ear_r_01, swing1LimitDegrees=10, swing2LimitDegrees=15, twistLimitDegrees=2}
 {bone1Name=dyn_templeFur_base_l, bone2Name=head, swing1LimitDegrees=20, swing2LimitDegrees=20, twistLimitDegrees=5}
 {bone1Name=dyn_templeFur_l_01, bone2Name=dyn_templeFur_base_l, swing1LimitDegrees=20, swing2LimitDegrees=20, twistLimitDegrees=5}
 {bone1Name=dyn_templeFur_l_02, bone2Name=dyn_templeFur_base_l, swing1LimitDegrees=20, swing2LimitDegrees=20, twistLimitDegrees=5}
 {bone1Name=dyn_templeFur_l_03, bone2Name=dyn_templeFur_base_l, swing1LimitDegrees=20, swing2LimitDegrees=20, twistLimitDegrees=5}
 {bone1Name=dyn_templeFur_base_r, bone2Name=head, swing1LimitDegrees=20, swing2LimitDegrees=20, twistLimitDegrees=5}
 {bone1Name=dyn_templeFur_r_01, bone2Name=dyn_templeFur_base_r, swing1LimitDegrees=20, swing2LimitDegrees=20, twistLimitDegrees=5}
 {bone1Name=dyn_templeFur_r_02, bone2Name=dyn_templeFur_base_r, swing1LimitDegrees=20, swing2LimitDegrees=20, twistLimitDegrees=5}
 {bone1Name=dyn_templeFur_r_03, bone2Name=dyn_templeFur_base_r, swing1LimitDegrees=20, swing2LimitDegrees=20, twistLimitDegrees=5}
 {bone1Name=dyn_eyeFur_l, bone2Name=head, swing1LimitDegrees=10, swing2LimitDegrees=5, twistLimitDegrees=2}
 {bone1Name=dyn_eyeFur_r, bone2Name=head, swing1LimitDegrees=10, swing2LimitDegrees=5, twistLimitDegrees=2}
 {
  bone1Name=upperarm_twist_01_l
  bone2Name=clavicle_l
  swing1LimitDegrees=45
  swing1Motion=Free
  swing2LimitDegrees=45
  swing2Motion=Free
  twistLimitDegrees=45
  twistMotion=Free
 }
 {
  bone1Name=upperarm_twist_01_r
  bone2Name=clavicle_r
  swing1LimitDegrees=45
  swing1Motion=Free
  swing2LimitDegrees=45
  swing2Motion=Free
  twistLimitDegrees=45
  twistMotion=Free
 }
]
