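# Secondary-motion AnimBlueprint for the face accessory mesh of the female Tiger Fashion body, animating temple braids, fur and hair.
# It copies the attached parent mesh's pose, runs a head-space RigidBody node with the temple braid physics asset, then chains spring-limited AnimDynamics nodes for temple fur and hair bones.
# Each update lowers the RigidBody's acceleration scale while crouching or targeting, and resets dynamics once whenever targeting toggles.
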AnimBlueprint /BRCosmetics/Characters/Player/Female/Medium/Bodies/F_MED_Tiger_Fashion/Meshes/Parts/F_MED_Tiger_Fashion_FaceAcc_AnimBP
target_skeleton=&/F_MED_Tiger_Fashion_FaceAcc_Skeleton
graphs [
 AnimGraph {nodes [
   n1 = RigidBody!(
    ComponentPose=Local To Component(LocalPose=Copy Pose From Mesh() {bUseAttachedParent=true})
    Alpha
    ExternalForce
    ComponentLinearAccScale:Vector=Get override_acc_scale() -> (override_acc_scale:Vector)
   ) -> (Pose) {
    bClampLinearTranslationLimitToRefPose=true
    bOverrideWorldGravity=true
    overridePhysicsAsset=&/F_MED_Tiger_Fashion_FaceAcc_TempleBraid_Physics1
    simulationSpace=BaseBoneSpace
    alphaBoolBlend {blendOption=Linear}
    baseBoneRef {boneName=head}
    componentAppliedLinearAccClamp=<3000 3000 3000>
    componentLinearAccScale=<0.3 0.25 0.3>
    componentLinearVelScale=<0.2 0.2 0.3>
    overrideWorldGravity {z=-980}
   }
   n2 = Anim Dynamics - Bone: dyn_templeFur_base_l(ComponentPose=@n1, Alpha=1, ExternalForce="0,0,0") {
    angularDampingOverride=0.999
    angularSpringConstant=2000
    bAngularSpring=true
    bOverrideAngularDamping=true
    bUseGravityOverride=true
    bUsePlanarLimit=false
    simulationSpace=BoneRelative
    alphaBoolBlend {blendOption=Linear}
    boundBone {boneName=dyn_templeFur_base_l}
    componentAppliedLinearAccClamp=<2000 2000 2000>
    componentLinearAccScale=<0.3 0.3 0.3>
    componentLinearVelScale=<0.3 0.3 0.3>
    gravityOverride {z=-800}
    physicsBodyDefinitions [{
      boundBone {boneName=dyn_templeFur_base_l}
      boxExtents=<7 7 7>
      constraintSetup {angularLimitsMax=<0 20 20>, angularLimitsMin=<0 -20 -10>, angularTarget=<0 0 0>}
      localJointOffset=<2 0 0>
     }]
   }
   n3 = Anim Dynamics - Bone: dyn_templeFur_l_01(ComponentPose=@n2, Alpha=1, ExternalForce="0,0,0") {
    angularDampingOverride=0.999
    angularSpringConstant=3000
    bAngularSpring=true
    bOverrideAngularDamping=true
    bUseGravityOverride=true
    bUsePlanarLimit=false
    simulationSpace=BoneRelative
    alphaBoolBlend {blendOption=Linear}
    boundBone {boneName=dyn_templeFur_l_01}
    componentAppliedLinearAccClamp=<2000 2000 2000>
    componentLinearAccScale=<0.3 0.3 0.3>
    componentLinearVelScale=<0.3 0.3 0.3>
    gravityOverride {z=-800}
    physicsBodyDefinitions [{
      boundBone {boneName=dyn_templeFur_l_01}
      boxExtents=<3 3 3>
      constraintSetup {angularLimitsMax=<5 10 10>, angularLimitsMin=<-5 -10 -10>, angularTarget=<0 0 0>}
      localJointOffset=<0.5 0 0>
     }]
   }
   n4 = Anim Dynamics - Bone: dyn_templeFur_l_02(ComponentPose=@n3, Alpha=1, ExternalForce="0,0,0") {
    angularDampingOverride=0.999
    angularSpringConstant=3000
    bAngularSpring=true
    bOverrideAngularDamping=true
    bUseGravityOverride=true
    bUsePlanarLimit=false
    simulationSpace=BoneRelative
    alphaBoolBlend {blendOption=Linear}
    boundBone {boneName=dyn_templeFur_l_02}
    componentAppliedLinearAccClamp=<2000 2000 2000>
    componentLinearAccScale=<0.3 0.3 0.3>
    componentLinearVelScale=<0.3 0.3 0.3>
    gravityOverride {z=-800}
    physicsBodyDefinitions [{
      boundBone {boneName=dyn_templeFur_l_02}
      boxExtents=<3 3 3>
      constraintSetup {angularLimitsMax=<5 10 10>, angularLimitsMin=<-5 -10 -10>, angularTarget=<0 0 0>}
      localJointOffset=<0.5 0 0>
     }]
   }
   n5 = Anim Dynamics - Bone: dyn_templeFur_l_03(ComponentPose=@n4, Alpha=1, ExternalForce="0,0,0") {
    angularDampingOverride=0.999
    angularSpringConstant=3000
    bAngularSpring=true
    bOverrideAngularDamping=true
    bUseGravityOverride=true
    bUsePlanarLimit=false
    simulationSpace=BoneRelative
    alphaBoolBlend {blendOption=Linear}
    boundBone {boneName=dyn_templeFur_l_03}
    componentAppliedLinearAccClamp=<2000 2000 2000>
    componentLinearAccScale=<0.3 0.3 0.3>
    componentLinearVelScale=<0.3 0.3 0.3>
    gravityOverride {z=-800}
    physicsBodyDefinitions [{
      boundBone {boneName=dyn_templeFur_l_03}
      boxExtents=<3 3 3>
      constraintSetup {angularLimitsMax=<5 10 10>, angularLimitsMin=<-5 -10 -10>, angularTarget=<0 0 0>}
      localJointOffset=<0.5 0 0>
     }]
   }
   n6 = Anim Dynamics - Bone: dyn_templeFur_base_r(ComponentPose=@n5, Alpha=1, ExternalForce="0,0,0") {
    angularDampingOverride=0.999
    angularSpringConstant=2000
    bAngularSpring=true
    bOverrideAngularDamping=true
    bUseGravityOverride=true
    bUsePlanarLimit=false
    simulationSpace=BoneRelative
    alphaBoolBlend {blendOption=Linear}
    boundBone {boneName=dyn_templeFur_base_r}
    componentAppliedLinearAccClamp=<2000 2000 2000>
    componentLinearAccScale=<0.3 0.3 0.3>
    componentLinearVelScale=<0.3 0.3 0.3>
    gravityOverride {z=-800}
    physicsBodyDefinitions [{
      boundBone {boneName=dyn_templeFur_base_r}
      boxExtents=<7 7 7>
      constraintSetup {angularLimitsMax=<0 20 15>, angularLimitsMin=<0 -20 -5>, angularTarget=<0 0 0>}
      localJointOffset=<2 0 0>
     }]
   }
   n7 = Anim Dynamics - Bone: dyn_templeFur_r_01(ComponentPose=@n6, Alpha=1, ExternalForce="0,0,0") {
    angularDampingOverride=0.999
    angularSpringConstant=3000
    bAngularSpring=true
    bOverrideAngularDamping=true
    bUseGravityOverride=true
    bUsePlanarLimit=false
    simulationSpace=BoneRelative
    alphaBoolBlend {blendOption=Linear}
    boundBone {boneName=dyn_templeFur_r_01}
    componentAppliedLinearAccClamp=<2000 2000 2000>
    componentLinearAccScale=<0.3 0.3 0.3>
    componentLinearVelScale=<0.3 0.3 0.3>
    gravityOverride {z=-800}
    physicsBodyDefinitions [{
      boundBone {boneName=dyn_templeFur_r_01}
      boxExtents=<3 3 3>
      constraintSetup {angularLimitsMax=<5 10 10>, angularLimitsMin=<-5 -10 -10>, angularTarget=<0 0 0>}
      localJointOffset=<0.5 0 0>
     }]
   }
   n8 = Anim Dynamics - Bone: dyn_templeFur_r_02(ComponentPose=@n7, Alpha=1, ExternalForce="0,0,0") {
    angularDampingOverride=0.999
    angularSpringConstant=3000
    bAngularSpring=true
    bOverrideAngularDamping=true
    bUseGravityOverride=true
    bUsePlanarLimit=false
    simulationSpace=BoneRelative
    alphaBoolBlend {blendOption=Linear}
    boundBone {boneName=dyn_templeFur_r_02}
    componentAppliedLinearAccClamp=<2000 2000 2000>
    componentLinearAccScale=<0.3 0.3 0.3>
    componentLinearVelScale=<0.3 0.3 0.3>
    gravityOverride {z=-800}
    physicsBodyDefinitions [{
      boundBone {boneName=dyn_templeFur_r_02}
      boxExtents=<3 3 3>
      constraintSetup {angularLimitsMax=<5 10 10>, angularLimitsMin=<-5 -10 -10>, angularTarget=<0 0 0>}
      localJointOffset=<0.5 0 0>
     }]
   }
   n9 = Anim Dynamics - Bone: dyn_templeFur_r_03(ComponentPose=@n8, Alpha=1, ExternalForce="0,0,0") {
    angularDampingOverride=0.999
    angularSpringConstant=3000
    bAngularSpring=true
    bOverrideAngularDamping=true
    bUseGravityOverride=true
    bUsePlanarLimit=false
    simulationSpace=BoneRelative
    alphaBoolBlend {blendOption=Linear}
    boundBone {boneName=dyn_templeFur_r_03}
    componentAppliedLinearAccClamp=<2000 2000 2000>
    componentLinearAccScale=<0.3 0.3 0.3>
    componentLinearVelScale=<0.3 0.3 0.3>
    gravityOverride {z=-800}
    physicsBodyDefinitions [{
      boundBone {boneName=dyn_templeFur_r_03}
      boxExtents=<3 3 3>
      constraintSetup {angularLimitsMax=<5 10 10>, angularLimitsMin=<-5 -10 -10>, angularTarget=<0 0 0>}
      localJointOffset=<0.5 0 0>
     }]
   }
   n10 = Anim Dynamics - Bone: dyn_hair_05(ComponentPose=@n9, Alpha=1, ExternalForce="0,0,0") {
    angularDampingOverride=0.99
    angularSpringConstant=5000
    bAngularSpring=true
    bGravityOverrideInSimSpace=true
    bOverrideAngularDamping=true
    numSolverIterationsPostUpdate=4
    numSolverIterationsPreUpdate=16
    alphaBoolBlend {blendOption=Linear}
    boundBone {boneName=dyn_hair_05}
    physicsBodyDefinitions [{
      sphereCollisionRadius=0
      boundBone {boneName=dyn_hair_05}
      boxExtents=<5 30 30>
      constraintSetup {angularTargetAxis=AxisZ, angularLimitsMax=<2 0 0>, angularLimitsMin=<-2 0 -20>, angularTarget=<0 0 1>}
      localJointOffset=<-1 0 0>
     }]
   }
   n11 = Anim Dynamics - Bone: dyn_hair_05_a(ComponentPose=@n10, Alpha=1, ExternalForce="0,0,0") {
    angularDampingOverride=0.99
    angularSpringConstant=5000
    bAngularSpring=true
    bGravityOverrideInSimSpace=true
    bOverrideAngularDamping=true
    numSolverIterationsPostUpdate=4
    numSolverIterationsPreUpdate=16
    alphaBoolBlend {blendOption=Linear}
    boundBone {boneName=dyn_hair_05_a}
    physicsBodyDefinitions [{
      sphereCollisionRadius=0
      boundBone {boneName=dyn_hair_05_a}
      boxExtents=<5 30 30>
      constraintSetup {angularTargetAxis=AxisZ, angularLimitsMax=<2 15 10>, angularLimitsMin=<-2 -6 -20>, angularTarget=<0 0 1>}
      localJointOffset=<-1 0 0>
     }]
   }
   n12 = Anim Dynamics - Bone: dyn_hair_05_b(ComponentPose=@n11, Alpha=1, ExternalForce="0,0,0") {
    angularDampingOverride=0.99
    angularSpringConstant=5000
    bAngularSpring=true
    bGravityOverrideInSimSpace=true
    bOverrideAngularDamping=true
    numSolverIterationsPostUpdate=4
    numSolverIterationsPreUpdate=16
    alphaBoolBlend {blendOption=Linear}
    boundBone {boneName=dyn_hair_05_b}
    physicsBodyDefinitions [{
      sphereCollisionRadius=0
      boundBone {boneName=dyn_hair_05_b}
      boxExtents=<5 30 30>
      constraintSetup {angularTargetAxis=AxisZ, angularLimitsMax=<2 20 20>, angularLimitsMin=<-2 -20 -20>, angularTarget=<0 0 1>}
      localJointOffset=<-1 0 0>
     }]
   }
   n13 = Anim Dynamics - Bone: dyn_hair_04(ComponentPose=@n12, Alpha=1, ExternalForce="0,0,0") {
    angularDampingOverride=0.99
    angularSpringConstant=5000
    bAngularSpring=true
    bGravityOverrideInSimSpace=true
    bOverrideAngularDamping=true
    numSolverIterationsPostUpdate=4
    numSolverIterationsPreUpdate=16
    alphaBoolBlend {blendOption=Linear}
    boundBone {boneName=dyn_hair_04}
    physicsBodyDefinitions [{
      sphereCollisionRadius=0
      boundBone {boneName=dyn_hair_04}
      boxExtents=<5 30 30>
      constraintSetup {angularTargetAxis=AxisZ, angularLimitsMax=<2 10 20>, angularLimitsMin=<-2 -20 -20>, angularTarget=<0 0 1>}
      localJointOffset=<-1 0 0>
     }]
   }
   n14 = Anim Dynamics - Bone: dyn_hair_04_a(ComponentPose=@n13, Alpha=1, ExternalForce="0,0,0") {
    angularDampingOverride=0.99
    angularSpringConstant=5000
    bAngularSpring=true
    bGravityOverrideInSimSpace=true
    bOverrideAngularDamping=true
    numSolverIterationsPostUpdate=4
    numSolverIterationsPreUpdate=16
    alphaBoolBlend {blendOption=Linear}
    boundBone {boneName=dyn_hair_04_a}
    physicsBodyDefinitions [{
      sphereCollisionRadius=0
      boundBone {boneName=dyn_hair_04_a}
      boxExtents=<5 30 30>
      constraintSetup {angularTargetAxis=AxisZ, angularLimitsMax=<2 -5 20>, angularLimitsMin=<-2 -20 -20>, angularTarget=<0 0 1>}
      localJointOffset=<-1 0 0>
     }]
   }
   Output Pose(Result=Component To Local(ComponentPose=@n14)) {layerGroup=None}
  ]}
 EventGraph {nodes [
   Event BlueprintUpdateAnimation()
   n1 = Get bIsCrouching()
   n2 = Get bIsTargeting()
   > Set override_acc_scale(override_acc_scale:Vector=SelectVector(A="0.05,0.05,0.05", B="0.35,0.3,0.35", bPickA=OR Boolean(A=@n1, B=@n2))) -> (Output_Get:Vector)
   n3 = Branch(Condition=Get bIsTargeting())
   > Do Once(Reset=@n5.then)
   > n4 = ResetDynamics(InTeleportType=ResetPhysics)
   Do Once(execute=@n3.else, Reset=@n4.then)
   > n5 = ResetDynamics(InTeleportType=ResetPhysics)
   TryGetPawnOwner()
  ]}
]
variables [override_acc_scale {type=Vector, default=<0.35 0.3 0.35>}]
